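# Physics asset for the gnome pickaxe melee weapon, giving its dangling light and flower secondary motion on the two-handed handle.
# Capsule bodies cover the handle, a flower base with five flower bones, and a two-link light chain.
# The flower base is rotationally locked to the handle, with each flower bone constrained to it using small angle values.
# The light handle hangs from the main handle, and the light pivots from it on a single unlocked swing axis.
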PhysicsAsset /BRCosmetics/Weapons/FORT_Melee/Pickaxe_Gnome/Meshes/Pickaxe_Gnome_Physics
bodies {
 Melee_TwoHanded_Handle [
  {shapeType=Capsule, length=81.5894, radius=2.7606, center=<0.3149 -0.6869 41.7103>, rotation=<rot 2.4475 -90.007 -0.6789>}
  {shapeType=Capsule, length=30, radius=2.75, center=<-0.0015 9.8718 76.9568>, rotation=<rot 0 0 114.0366>}
 ]
 dyn_flower_01 [{shapeType=Capsule, length=2.5, radius=2, center=<2 0 0>, rotation=<rot 90 0 0>}]
 dyn_flower_02 [{shapeType=Capsule, length=2.5, radius=2, center=<2 0 0>, rotation=<rot 90 0 0>}]
 dyn_flower_03 [{shapeType=Capsule, length=2.5, radius=2, center=<2 0 0>, rotation=<rot 90 0 0>}]
 dyn_flower_04 [{shapeType=Capsule, length=2.5, radius=2, center=<2 0 0>, rotation=<rot 90 0 0>}]
 dyn_flower_05 [{shapeType=Capsule, length=2.5, radius=2, center=<2 0 0>, rotation=<rot 90 0 0>}]
 dyn_flower_base [{shapeType=Capsule, length=0, radius=3.4188, center=<-0.2023 -0.1567 -0.2937>, rotation=<rot -0.0001 67.363 -179.0499>}]
 dyn_light [{shapeType=Capsule, length=14.9107, radius=9.0327, center=<12.5014 -1.0321 -0.4635>, rotation=<rot 0 -91.0168 -87.8548>}]
 dyn_lightHandle [{shapeType=Capsule, length=16.6493, radius=2.9472, center=<9.9725 0.7718 0.5351>, rotation=<rot 4.3984 -86.959 -88.7989>}]
}
constraints [
 {
  bone1Name=dyn_flower_base
  bone2Name=Melee_TwoHanded_Handle
  swing1LimitDegrees=45
  swing1Motion=Locked
  swing2LimitDegrees=45
  swing2Motion=Locked
  twistLimitDegrees=45
  twistMotion=Locked
 }
 {bone1Name=dyn_lightHandle, bone2Name=Melee_TwoHanded_Handle, swing1LimitDegrees=15, swing2LimitDegrees=60, twistLimitDegrees=5}
 {bone1Name=dyn_flower_01, bone2Name=dyn_flower_base, swing1LimitDegrees=5, swing2LimitDegrees=15, twistLimitDegrees=5}
 {bone1Name=dyn_flower_02, bone2Name=dyn_flower_base, swing1LimitDegrees=5, swing2LimitDegrees=15, twistLimitDegrees=5}
 {bone1Name=dyn_flower_03, bone2Name=dyn_flower_base, swing1LimitDegrees=5, swing2LimitDegrees=15, twistLimitDegrees=5}
 {bone1Name=dyn_flower_04, bone2Name=dyn_flower_base, swing1LimitDegrees=5, swing2LimitDegrees=15, twistLimitDegrees=5}
 {bone1Name=dyn_flower_05, bone2Name=dyn_flower_base, swing1LimitDegrees=5, swing2LimitDegrees=15, twistLimitDegrees=5}
 {
  bone1Name=dyn_light
  bone2Name=dyn_lightHandle
  swing1LimitDegrees=60
  swing2LimitDegrees=45
  swing2Motion=Locked
  twistLimitDegrees=45
  twistMotion=Locked
 }
]
