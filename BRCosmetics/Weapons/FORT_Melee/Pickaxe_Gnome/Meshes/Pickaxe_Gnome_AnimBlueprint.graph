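# Animation Blueprint for the gnome pickaxe that adds secondary motion to its dynamic light bones on top of the input pose.
# When bRigidBodyEnabled is set, a RigidBody node simulates the override physics asset relative to the handle bone.
# Otherwise a Trail controller drives dyn_light, constrained by planar limits on dyn_lightHandle, whose preset rotation switches on bIsInFrontEnd.
# Both branches read one cached input pose.
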AnimBlueprint /BRCosmetics/Weapons/FORT_Melee/Pickaxe_Gnome/Meshes/Pickaxe_Gnome_AnimBlueprint
target_skeleton=&/Pickaxe_Gnome_Skeleton
graphs [
 AnimGraph {nodes [
   n1 = "Transform (Modify) Bone - Bone: dyn_lightHandle"(ComponentPose=Local To Component(LocalPose=Use cached pose 'Start Cache'()), Alpha=1, Translation="0,0,0", Rotation="-70,0,0", Scale="1,1,1") {
    rotationMode=BMM_Replace
    rotationSpace=BCS_BoneSpace
    translationMode=BMM_Additive
    translationSpace=BCS_BoneSpace
    alphaBoolBlend {blendOption=Linear}
    boneToModify {boneName=dyn_lightHandle}
    rotation {pitch=-70}
   }
   n2 = Component To Local(ComponentPose=@n1)
   n3 = "Transform (Modify) Bone - Bone: dyn_lightHandle"(ComponentPose=Local To Component(LocalPose=Use cached pose 'Start Cache'()), Alpha=1, Translation="0,0,0", Rotation="10,25,0", Scale="1,1,1") {
    rotationMode=BMM_Replace
    rotationSpace=BCS_BoneSpace
    translationMode=BMM_Additive
    translationSpace=BCS_BoneSpace
    alphaBoolBlend {blendOption=Linear}
    boneToModify {boneName=dyn_lightHandle}
    rotation {pitch=10, yaw=25}
   }
   n4 = Component To Local(ComponentPose=@n3)
   n5 = Get bIsInFrontEnd()
   n6 = Trail controller - Bone: dyn_light(
    ComponentPose=Local To Component(LocalPose=Blend Poses by bool(BlendPose_0=@n4, BlendPose_1=@n2, bActiveValue=@n5) {blendType=Linear})
    Alpha=1
   ) {
    bActorSpaceFakeVel=true
    bLimitRotation=true
    bShowBaseMotion=false
    bShowLimit=false
    bUsePlanarLimit=true
    baseJoint {boneName=Melee_TwoHanded_Handle}
    fakeVelocity=<1 1 1>
    planarLimits [
     {drivingBone {boneName=dyn_lightHandle}, planeTransform {location=<15 0 0>, rotation=<rot -90 0 0>, scale=<1 1 1>}}
     {drivingBone {boneName=dyn_lightHandle}, planeTransform {location=<0 0 0>, rotation=<rot 0 0 90>, scale=<1 1 1>}}
     {drivingBone {boneName=dyn_lightHandle}, planeTransform {location=<0 0 -10>, rotation=<rot 0 0 0>, scale=<1 1 1>}}
     {drivingBone {boneName=dyn_lightHandle}, planeTransform {location=<0 0 10>, rotation=<rot 0 0 180>, scale=<1 1 1>}}
     {drivingBone {boneName=dyn_lightHandle}, planeTransform {location=<0 10 0>, rotation=<rot 0 0 -90>, scale=<1 1 1>}}
     {drivingBone {boneName=None}, planeTransform {location=<0 0 0>, rotation=<rot 0 0 0>, scale=<1 1 1>}}
    ]
    rotationLimits [{limitMax=<180 180 180>}, {limitMax=<180 180 180>}]
    rotationOffsets [<0 0 0>, <0 0 0>]
    trailBone {boneName=dyn_light}
   }
   n7 = RigidBody(ComponentPose=Local To Component(LocalPose=Use cached pose 'Start Cache'())) {
    bClampLinearTranslationLimitToRefPose=true
    bOverrideWorldGravity=true
    overridePhysicsAsset=&/Pickaxe_Gnome_Physics
    simulationSpace=BaseBoneSpace
    alphaBoolBlend {blendOption=Linear}
    baseBoneRef {boneName=Melee_TwoHanded_Handle}
    componentAppliedLinearAccClamp=<3000 3000 3000>
    componentLinearAccScale=<0.3 0.3 0.3>
    componentLinearVelScale=<0.3 0.3 0.3>
    overrideWorldGravity {z=-980}
   }
   Output Pose(Result=Blend Poses by bool(BlendPose_0=Component To Local(ComponentPose=@n7), BlendPose_1=Component To Local(ComponentPose=@n6), bActiveValue=Get bRigidBodyEnabled()) {blendType=Linear}) {layerGroup=None}
   AnimGraphNode_SubInput_0 = Input Pose()
   Save cached pose 'Start Cache'(Pose=@AnimGraphNode_SubInput_0) {cacheName=Start Cache}
   Get bIsInFrontEnd()
  ]}
 EventGraph {nodes [Event BlueprintUpdateAnimation()]}
]
variables []
